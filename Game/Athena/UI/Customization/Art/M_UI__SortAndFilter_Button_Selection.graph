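# Translucent UI material that draws the selection look for a Sort and Filter button in the customization screen.
# A triangle-grid texture, mapped to aspect-corrected screen UVs, rotated and slowly panned, blends two blue tints.
# A horizontal gradient, positioned by the Burst parameter, fades between white and that pattern.
# A sharpened, rotated mask on 9-slice UVs adds a colored angled bar on the left.
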Material /Game/Athena/UI/Customization/Art/M_UI__SortAndFilter_Button_Selection
graph {
 properties {
  bUsedWithStaticMesh=false
  blendMode=BLEND_Translucent
  materialDomain=MD_UI
  pixelDepthOffsetMode=PDOM_Legacy
  usedShadingModels=Default Lit
 }
 nodes [
  n1 = VectorParameter() {group=Triangles, parameterName=RGB1, defaultValue=<rgba 0.6038 0.8879 1 1>}
  n2 = VectorParameter() {group=Triangles, parameterName=RGB2, defaultValue=<rgba 0.0331 0.402 1 1>}
  n3 = MaterialFunctionCall(Float2=MaterialFunctionCall() {materialFunction=&ScreenResolution}.Visible Resolution) {materialFunction=&BreakOutFloat2Components}
  n4 = AppendVector(A=Divide(A=@n3.R, B=Max(A=@n3.G)), B=Constant() {r=1})
  n5 = MaterialFunctionCall() {materialFunction=&ScreenAlignedUVs}
  n6 = Panner(
   Coordinate=Rotator(Coordinate=Multiply(A=@n4, B=@n5."X 100%, Y 100%"), Time=ScalarParameter() {defaultValue=-0.0571, group=Triangles, parameterName=Tri Angle}) {speed=6.2832}
   Speed=AppendVector(A=ScalarParameter() {defaultValue=0.0025, group=Triangles, parameterName=PanSpeed}, B=Constant())
  )
  n7 = ScalarParameter() {defaultValue=5, group=Triangles, parameterName=Tri Tile Amount}
  n8 = Multiply(
   A=TextureSample(UVs=Multiply(A=@n6, B=@n7)) {texture=/Game/Global/Textures/Default/Masks_and_Gradients/T-TriangleGrid}.R
   B=ScalarParameter() {defaultValue=0.2, group=Triangles, parameterName=TriangleDensity}
  )
  n9 = MaterialFunctionCall() {materialFunction=&GetUserInterfaceUV}
  n10 = LinearInterpolate(
   A=ScalarParameter() {group=white gradient, parameterName=white_Burst_start}
   B=ScalarParameter() {defaultValue=1, group=white gradient, parameterName=white_Burst_end}
   Alpha=ScalarParameter() {group=white gradient, parameterName=Burst}
  )
  n11 = Multiply(
   A=ComponentMask(None=Subtract(A=@n9.Normalized UV, B=@n10)) {r=true}
   B=ScalarParameter() {defaultValue=1.5, group=white gradient, parameterName=White_sharpness}
  )
  n12 = MaterialFunctionCall() {materialFunction=&GetUserInterfaceUV}
  n13 = ScalarParameter() {defaultValue=-0.25, group=Left Bar, parameterName=LeftBar_Position}
  n14 = ComponentMask(None=Rotator(Coordinate=Add(A=@n12.9-Slice UV, B=@n13), Time=ScalarParameter() {defaultValue=-0.25, group=Left Bar, parameterName=LeftBar_Angle})) {r=true}
  n15 = ScalarParameter() {defaultValue=256, group=Left Bar, parameterName=LeftBar_Sharpness}
  LinearInterpolate(
   A=VectorParameter() {group=Left Bar, parameterName=LeftBar_Color, defaultValue=<rgba 0.9301 1 0.0103 1>}.RGB
   B=LinearInterpolate(A=Constant() {r=1}, B=LinearInterpolate(A=@n1.RGB, B=@n2.RGB, Alpha=@n8), Alpha=Saturate(None=@n11))
   Alpha=Saturate(None=Multiply(A=@n14, B=@n15))
  )
 ]
}
